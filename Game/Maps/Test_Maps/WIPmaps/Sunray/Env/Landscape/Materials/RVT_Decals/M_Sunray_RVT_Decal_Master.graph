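# Master translucent material for Sunray landscape decals that writes base color, specular, roughness, normal and opacity to a Runtime Virtual Texture.
# Procedural UV, on by default, offsets texture coordinates by the rounded fraction of world Z times UV Slices, selecting texture slices by height.
# Static switches choose SMR and normal textures or fallback values.
# Opacity uses diffuse alpha, optionally masked by red vertex color, scaled by either Opacity or fractional world height.
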Material /Game/Maps/Test_Maps/WIPmaps/Sunray/Env/Landscape/Materials/RVT_Decals/M_Sunray_RVT_Decal_Master
graph {
 properties {blendMode=BLEND_Translucent, pixelDepthOffsetMode=PDOM_Legacy, usedShadingModels=Default Lit}
 nodes [
  n1 = WorldPosition()
  n2 = Round(None=Frac(None=ComponentMask(None=@n1.XYZ) {b=true}))
  n3 = ScalarParameter() {defaultValue=0.5, group=Procedural UV, parameterName=UV Slices, sliderMax=0.05, sliderMin=1}
  n4 = TextureCoordinate()
  n5 = StaticSwitchParameter(True=Add(A=Multiply(A=@n2, B=@n3) {constB=0.5}, B=@n4), False=@n4) {defaultValue=true, group=Procedural UV, parameterName=Procedural UV}
  n6 = TextureSampleParameter2D(UVs=@n5) {parameterName=Diffuse}
  n7 = TextureSampleParameter2D(UVs=@n5) {parameterName=SMR}
  n8 = Frac(None=ComponentMask(None=WorldPosition().XYZ) {b=true})
  n9 = StaticSwitchParameter(True=Multiply(A=VertexColor().R, B=@n6.A), False=@n6.A) {group=Opacity, parameterName=Blend Red Vert Color}
  n10 = ScalarParameter() {defaultValue=1, group=Opacity, parameterName=Opacity, sliderMax=2}
  RuntimeVirtualTextureOutput!(
   BaseColor=@n6.RGB
   Specular=StaticSwitchParameter(True=@n7.R, False=ScalarParameter() {defaultValue=0.2, parameterName=Specular, sliderMax=1}) {group=Texture Slots, parameterName=Specular Texture}
   Roughness=StaticSwitchParameter(True=@n7.B, False=ScalarParameter() {defaultValue=1, parameterName=Roughness, sliderMax=1}) {group=Texture Slots, parameterName=Roughness Texture}
   Normal=StaticSwitchParameter(True=TextureSampleParameter2D(UVs=@n5) {parameterName=Normal}.RGB, False=Constant3Vector() {constant=<rgba 0 0 1 0>}) {group=Texture Slots, parameterName=Normal Texture}
   WorldHeight
   Opacity=StaticSwitchParameter(True=Multiply(A=@n8, B=@n9), False=Multiply(A=@n9, B=@n10)) {group=Opacity, parameterName=Opacity Frac Height}
   Mask
   Displacement
   Mask4
  )
 ]
}
